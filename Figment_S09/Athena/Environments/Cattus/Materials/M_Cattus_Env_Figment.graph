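# Master material for Cattus POI environment meshes, with static switches selecting Doggus sword, world-aligned bone, or teeth with hit-effect vertex offset.
# The graph holds two chains: one height-blends gravel and grass using vertex color and MPC_CattusPOI heights, the other adds dust over simpler surfaces.
# Normals output in world space.
# Both chains blend into Apollo landscape runtime virtual textures by height above terrain, bypassed on mobile, and include Creative move-tool support.
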Material /Figment_S09/Athena/Environments/Cattus/Materials/M_Cattus_Env_Figment
graph {
 properties {
  bTangentSpaceNormal=false
  bUseMaterialAttributes=true
  bUsedWithNanite=true
  pixelDepthOffsetMode=PDOM_Legacy
  usedShadingModels=Default Lit
 }
 nodes [
  n1 = ScalarParameter() {defaultValue=2, parameterName=BoneProjectionContrast}
  n2 = MaterialFunctionCall!(
   Diffuse Texture Object=TextureObject() {texture=/Game/Athena/Environments/Cattus/Textures/T_Bone_D}
   Spec Texture Object=TextureObject() {samplerType=SAMPLERTYPE_Masks, texture=/Game/Athena/Environments/Cattus/Textures/T_Bone_M}
   TextureSize=ScalarParameter() {defaultValue=369.8548, parameterName=SizeBone_Tiling}
   --------------------------
   Use seperate settings for Z projection
   Use seperate texture for Z projection
   Texture Size Z projection
   Z Plane Diffuse Texture Object
   Z Plane Spec Texture Object
   "--------------------------      "
   ProjectionTransitionContrast=@n1
   Use Texture Alpha Heightmap to Refine the Transistion
   HeightLerpTransitionContrast
   PickTheDiffuseHeightChannel
   TheDiffuseHeightChannel
   "--------------------------       "
   NormalTextureObject=TextureObject() {samplerType=SAMPLERTYPE_Normal, texture=/Game/Athena/Environments/Cattus/Textures/T_Bone_N}
   World Space Normals=StaticBool()
   UseUncorrectedNormalBlend
   "--------------------------  "
   Export Float 4 Diffuse
   Export Float 4 Spec
   "--------------------------     "
   WorldPosition
   World Space Normal
  ) {materialFunction=/Game/Packages/Fortress_SharedMaterials/Base_Material_Functions/WorldAlignedTextures_Complex_OptimzedNormal}
  n3 = MaterialFunctionCall(
   TextureSize=ScalarParameter() {defaultValue=1069.0742, parameterName=SizeBone_Tiling2}
   ProjectionTransitionContrast=@n1
   NormalTextureObject=TextureObject() {samplerType=SAMPLERTYPE_Normal, texture=&T_Bone_N}
   World Space Normals=StaticBool()
  ) {materialFunction=&WorldAlignedTextures_Complex_OptimzedNormal}
  Add(A=@n2.Normal XYZ Texture, B=Multiply(A=@n3.Normal XYZ Texture, B=Constant3Vector() {constant=<rgba 1 1 0 0>}))
  n4 = Transform(None=Normalize(VectorInput=VertexNormalWS())) {transformSourceType=TRANSFORMSOURCE_World, transformType=TRANSFORM_Tangent}
  AppendVector(A=ComponentMask(None=@n4) {g=true, r=true})
  Comment() {text=RVT}
  n5 = TextureSampleParameter2D() {
   parameterName=Basecolor
   samplerSource=SSM_Wrap_WorldGroupSettings
   texture=/Game/Athena/Prototype/Blueprints/Doggus_Sword/Textures/T_Doggus_Sword_D
  }
  n6 = TextureSample() {samplerSource=SSM_Wrap_WorldGroupSettings, texture=/Game/Athena/Prototype/Blueprints/Doggus_Sword/Textures/T_Doggus_Sword_E}
  n7 = BumpOffset(
   Coordinate=Multiply(A=TextureCoordinate(), B=ScalarParameter() {defaultValue=5, parameterName=TilingCracks})
   Height=Constant() {r=1}
   HeightRatioInput=LinearInterpolate(
    A=ScalarParameter() {defaultValue=-0.0476, parameterName=BumpOffsetCracks1}
    B=ScalarParameter() {defaultValue=-0.1524, parameterName=BumpOffsetCracks2}
    Alpha=MaterialFunctionCall() {materialFunction=&DitherTemporalAA}.Result
   )
  )
  n8 = Add(
   A=@n6.RGB
   B=Multiply(
    A=VectorParameter() {parameterName=CrackColor, defaultValue=<rgba 0.1927 0.0825 0.187 1>}.RGB
    B=TextureSample(UVs=@n7) {samplerSource=SSM_Wrap_WorldGroupSettings, texture=/Game/Effects/Fort_Effects/Textures/T_Cracks_Gradient}.R
   )
  )
  n9 = TextureSample() {
   samplerSource=SSM_Wrap_WorldGroupSettings
   samplerType=SAMPLERTYPE_Masks
   texture=/Game/Athena/Prototype/Blueprints/Doggus_Sword/Textures/T_Doggus_Sword_FX
  }
  n10 = LinearInterpolate(A=@n5.RGB, B=ShadingPathSwitch(Default=@n8, Deferred=@n8, Forward=@n6.RGB, Mobile=@n6.RGB), Alpha=@n9.R)
  n11 = CollectionParameter() {collection=/Game/Athena/Environments/Cattus/Materials/MPC_CattusPOI, parameterName=SwordHeightBlend_Offset}
  n12 = Subtract(A=ComponentMask(None=WorldPosition().XYZ) {b=true}, B=MaterialProxyReplace(Realtime=@n11, MaterialProxy=Subtract(A=@n11) {constB=768}))
  n13 = CollectionParameter() {collection=&MPC_CattusPOI, parameterName=SwordHeightBlend_Distance}
  n14 = Saturate(None=Add(A=Divide(A=@n12, B=@n13), B=ScalarParameter() {defaultValue=0.1905, parameterName=SwordBlendOpacity}))
  n15 = TextureSampleParameter2D() {
   parameterName=Specular
   samplerSource=SSM_Wrap_WorldGroupSettings
   samplerType=SAMPLERTYPE_Masks
   texture=/Game/Athena/Prototype/Blueprints/Doggus_Sword/Textures/T_Doggus_Sword_S
  }
  n16 = TextureSample(UVs=Multiply(A=TextureCoordinate(), B=ScalarParameter() {defaultValue=7.1097, parameterName=TilingDetail})) {samplerSource=SSM_Wrap_WorldGroupSettings, samplerType=SAMPLERTYPE_Masks, texture=/Game/Global/Textures/Default/Noise/T_Metal_Noise}
  n17 = LinearInterpolate(
   A=LinearInterpolate(A=ScalarParameter() {parameterName=RoughnessMin}, B=ScalarParameter() {defaultValue=1, parameterName=RoughnessMax}, Alpha=@n15.B)
   B=LinearInterpolate(A=ScalarParameter() {parameterName=RoughnessMetalMin}, B=ScalarParameter() {defaultValue=1, parameterName=RoughnessMetalMax}, Alpha=@n15.B)
   Alpha=@n15.G
  )
  n18 = Add(A=Multiply(A=@n16.R, B=ScalarParameter() {defaultValue=0.13, parameterName=RoughnessNoise}), B=@n17)
  n19 = MakeMaterialAttributes!(
   BaseColor=LinearInterpolate(
    A=VectorParameter() {parameterName=ColorSwordBlend, defaultValue=<rgba 0.2604 0.2048 0.1397 1>}.RGB
    B=StaticSwitchParameter(True=@n10, False=@n5.RGB) {defaultValue=true, parameterName=UseEmissiveBlend}
    Alpha=@n14
   )
   Metallic=@n15.G
   Specular=@n15.G
   Roughness=ShadingPathSwitch(Default=@n18, Deferred=@n18, Forward=@n17, Mobile=@n17)
   Anisotropy
   EmissiveColor
   Opacity
   OpacityMask
   Normal
   Tangent
   WorldPositionOffset
   SubsurfaceColor
   ClearCoat
   ClearCoatRoughness
   AmbientOcclusion
   Refraction
   CustomizedUVs_0
   CustomizedUVs_1
   CustomizedUVs_2
   CustomizedUVs_3
   CustomizedUVs_4
   CustomizedUVs_5
   CustomizedUVs_6
   CustomizedUVs_7
   PixelDepthOffset
   ShadingModel
   Displacement
  )
  n20 = Normalize(VectorInput=Transform(None=VectorParameter() {parameterName=GradientDirection, defaultValue=<rgba 0 0 1 1>}.RGB) {transformSourceType=TRANSFORMSOURCE_Local})
  n21 = Subtract(A=WorldPosition().XYZ, B=ObjectPositionWS())
  n22 = ObjectRadius()
  n23 = ScalarParameter() {defaultValue=2, parameterName=Gradient}
  n24 = Add(A=Divide(A=DotProduct(A=@n20, B=@n21), B=Multiply(A=@n22, B=@n23)), B=ScalarParameter() {defaultValue=0.7, parameterName=Gradientoffset}) {constB=0.5}
  n25 = TextureSampleParameter2D() {
   parameterName=Mask
   samplerSource=SSM_Wrap_WorldGroupSettings
   samplerType=SAMPLERTYPE_Masks
   texture=/Game/Athena/Environments/Cattus/Demo/Textures/Cattus_Skull_Outside_M
  }
  n26 = LinearInterpolate(
   A=LinearInterpolate(
    A=VectorParameter() {parameterName=Color1, defaultValue=<rgba 0.3698 0.2375 0.089 1>}.RGB
    B=VectorParameter() {parameterName=Color2, defaultValue=<rgba 0.5747 0.5747 0.5747 1>}.RGB
    Alpha=@n25.G
   )
   B=VectorParameter() {parameterName=Color3, defaultValue=<rgba 0.1892 0.1685 0.1277 1>}.RGB
   Alpha=@n25.B
  )
  n27 = LinearInterpolate(
   A=@n26
   B=Multiply(A=@n26, B=ScalarParameter() {defaultValue=2, parameterName=FresnelBrightness})
   Alpha=Fresnel(ExponentIn=ScalarParameter() {defaultValue=1.3, parameterName=FresnelBone}, Normal=VertexNormalWS()) {baseReflectFraction=0}
  )
  n28 = MaterialFunctionCall(Float3=@n2.XYZ Spec) {materialFunction=&BreakOutFloat3Components}
  n29 = MakeMaterialAttributes(
   BaseColor=Multiply(A=ShadingPathSwitch(Default=@n27, Deferred=@n27, Forward=@n26, Mobile=@n26), B=@n2.XYZ Diffuse Texture)
   Specular=@n28.R
   Roughness=Multiply(A=ScalarParameter() {defaultValue=3, parameterName=RoughnessBone}, B=@n28.B)
   Normal=@n2.Normal XYZ Texture
  )
  n30 = StaticSwitchParameter(
   True=MaterialFunctionCall!(
    Material=@n29
    Tint=VectorParameter() {parameterName=TeethGradientColor, defaultValue=<rgba 0.2344 0.2219 0.1807 1>}.RGB
    TintMask=Clamp(None=@n24)
   ) {materialFunction=/Engine/Functions/MaterialLayerFunctions/MatLayerBlend_Tint}.Blended Material
   False=@n29
  ) {parameterName=IsTeeth}
  n31 = CollectionParameter() {collection=&MPC_CattusPOI, parameterName=CattusHeightBlend_Offset}
  n32 = Subtract(
   A=ComponentMask(None=WorldPosition().XYZ) {b=true}
   B=MaterialProxyReplace(Realtime=@n31, MaterialProxy=ScalarParameter() {defaultValue=456.2168, parameterName=BlendOffset})
  )
  n33 = MaterialProxyReplace(
   Realtime=CollectionParameter() {collection=&MPC_CattusPOI, parameterName=CattusHeightBlend_Distance}
   MaterialProxy=Constant() {r=5500}
  )
  n34 = OneMinus(None=Saturate(None=Divide(A=@n32, B=@n33)))
  n35 = MaterialFunctionCall(
   Diffuse Texture Object=TextureObject() {samplerType=SAMPLERTYPE_Masks, texture=/Game/Athena/Environments/Cattus/Textures/T_Bone_M2}
   TextureSize=ScalarParameter() {defaultValue=1529.9003, parameterName=VinesScale}
   ProjectionTransitionContrast=ScalarParameter() {defaultValue=-0.3048, parameterName=VinesProjectionContrast}
   NormalTextureObject=TextureObject() {samplerType=SAMPLERTYPE_Normal, texture=/Game/Athena/Environments/Cattus/Textures/T_Bone_N2}
   World Space Normals=StaticBool()
   Export Float 4 Diffuse=StaticBool()
  ) {materialFunction=&WorldAlignedTextures_Complex_OptimzedNormal}
  n36 = Add(
   A=Subtract(A=VertexColor().G) {constB=0.5}
   B=MaterialFunctionCall(Transition Phase=@n34, Height Texture=@n35.XY Diffuse Texture, Contrast=ScalarParameter() {parameterName=ContrastVines}) {materialFunction=&HeightLerp}.Alpha
  )
  n37 = LinearInterpolate(
   A=ScalarParameter() {defaultValue=-1.2516, parameterName=GrassBlendDirtContrast}
   B=ScalarParameter() {defaultValue=0.8471, parameterName=GrassBlendDirtContrast2}
   Alpha=@n36
  )
  n38 = ScalarParameter() {defaultValue=455.5468, parameterName=SizeForestFloor}
  n39 = MaterialFunctionCall(
   Diffuse Texture Object=TextureObject() {texture=/Game/Packages/DS_Fortnight/WhiteRockCustom_D}
   TextureSize=@n38
   Texture Size Z projection=@n38
   NormalTextureObject=TextureObject() {samplerType=SAMPLERTYPE_Normal, texture=/Game/Packages/DS_Fortnight/WhiteRockCustom_N}
   World Space Normals=StaticBool()
   Export Float 4 Diffuse=StaticBool() {value=true}
  ) {materialFunction=&WorldAlignedTextures_Complex_OptimzedNormal}
  n40 = Multiply(
   A=LinearInterpolate(
    A=VectorParameter() {parameterName=ColorGravel, defaultValue=<rgba 0.5833 0.4972 0.4345 1>}.RGB
    B=VectorParameter() {parameterName=ColorGravel2, defaultValue=<rgba 0.2188 0.114 0.0376 1>}.RGB
    Alpha=@n37
   )
   B=ComponentMask(None=@n39.XYZ Diffuse Texture) {b=true, g=true, r=true}
  )
  n41 = ScalarParameter() {defaultValue=0.0857, parameterName=SpecularDirt}
  n42 = LinearInterpolate(
   A=ScalarParameter() {defaultValue=0.4476, parameterName=Roughness_dirt}
   B=ScalarParameter() {defaultValue=1, parameterName=Roughness_dirt2}
   Alpha=ComponentMask(None=@n39.XYZ Diffuse Texture) {r=true}
  )
  n43 = VertexNormalWS()
  n44 = Add(A=Subtract(A=VertexColor().R) {constB=0.5}, B=Saturate(None=@n37))
  n45 = LinearInterpolate(
   A=ScalarParameter() {parameterName=Grave_Transition_low}
   B=ScalarParameter() {defaultValue=0.8389, parameterName=Gravel_Transition_high}
   Alpha=ComponentMask(None=Add(A=@n43, B=@n44)) {b=true}
  )
  n46 = ComponentMask(None=@n39.XYZ Diffuse Texture) {a=true}
  n47 = ScalarParameter() {defaultValue=-0.019, parameterName=Gravel_TranstionContrast}
  n48 = MaterialFunctionCall(
   Diffuse Texture Object=TextureObject() {texture=/Game/Athena/Environments/Landscape/Textures/T_Athena_Grass_Farm_ColorMatched_D_2}
   TextureSize=ScalarParameter() {defaultValue=500, parameterName=GrassSize}
   World Space Normals=StaticBool()
   Export Float 4 Diffuse=StaticBool() {value=true}
  ) {materialFunction=&WorldAlignedTextures_Complex_OptimzedNormal}
  n49 = ComponentMask(None=@n48.XYZ Diffuse Texture) {b=true, g=true, r=true}
  n50 = MaterialFunctionCall!(In, Rock_SlopeMask, GrassMask, BaseColorTintOnlybyVertColor=@n49, VertexColorMask=VertexColor().B) {materialFunction=/Game/Athena/Environments/Cattus/MaterialFunctions/MF_CattusSpiral}
  n51 = ScalarParameter() {defaultValue=0.2381, parameterName=SpecularGrass1}
  n52 = ScalarParameter() {defaultValue=3, parameterName=SpecularGrass2}
  n53 = ComponentMask(None=@n48.XYZ Diffuse Texture) {r=true}
  n54 = ScalarParameter() {defaultValue=1, parameterName=Roughness_Grass}
  n55 = ScalarParameter() {defaultValue=-1, parameterName=Roughness_Grass2}
  n56 = ScalarParameter() {defaultValue=0.6667, parameterName=MaxGrass}
  n57 = MaterialFunctionCall!(
   Base Material=MaterialFunctionCall(
    Base Material=StaticSwitchParameter(True=@n19, False=@n30) {parameterName=IsSword}
    Top Material=MakeMaterialAttributes(BaseColor=@n40, Specular=@n41, Roughness=@n42, Normal=@n39.Normal XYZ Texture)
    Alpha=MaterialFunctionCall(Transition Phase=@n45, Height Texture=@n46, Contrast=@n47) {materialFunction=&HeightLerp}.Alpha
   ) {materialFunction=/Engine/Functions/MaterialLayerFunctions/MatLayerBlend_Standard}.Blended Material
   Top Material=MakeMaterialAttributes(
    BaseColor=StaticSwitchParameter(True=@n50.BaseColorOnlyOutput, False=@n49) {defaultValue=true, parameterName=UsePurpleGrowth}
    Specular=LinearInterpolate(A=@n51, B=@n52, Alpha=@n53)
    Roughness=LinearInterpolate(A=@n54, B=@n55, Alpha=@n53)
   )
   Alpha=MaterialFunctionCall(
    Transition Phase=Multiply(A=@n36, B=@n56)
    Height Texture=ComponentMask(None=@n48.XYZ Diffuse Texture) {a=true}
    Contrast=ScalarParameter() {defaultValue=1.0788, parameterName=TranstionContrastGrass}
   ) {materialFunction=&HeightLerp}.Alpha
  ) {materialFunction=&MatLayerBlend_Standard}
  n58 = GetMaterialAttributes(None=@n57.Blended Material) {attributeGetTypes [#1]}
  n59 = Multiply(A=@n4, B=VectorParameter() {parameterName=NormalIntensity, defaultValue=<rgba 1 1 0 1>}.RGB)
  n60 = TextureSampleParameter2D() {parameterName=Normal, samplerType=SAMPLERTYPE_Normal, texture=/Game/Athena/Prototype/Blueprints/Doggus_Sword/Textures/T_Doggus_Sword_N}
  n61 = Transform(None=@n60.RGB)
  n62 = SetMaterialAttributes!(
   MaterialAttributes=@n58.MaterialAttributes
   Normal=MaterialProxyReplace(Realtime=Add(A=@n61, B=@n59), MaterialProxy=Constant3Vector() {constant=<rgba 0 0 1 0>})
  ) {attributeSetTypes [#1]}
  n63 = GetMaterialAttributes(None=@n62) {attributeGetTypes [#2]}
  n64 = TextureSampleParameter2D() {
   parameterName=Mask
   samplerSource=SSM_Wrap_WorldGroupSettings
   samplerType=SAMPLERTYPE_Masks
   texture=&Cattus_Skull_Outside_M
  }
  n65 = Saturate(None=Add(A=@n64.R, B=ScalarParameter() {parameterName=AOintensity}))
  n66 = StaticSwitchParameter(
   True=SetMaterialAttributes!(MaterialAttributes=@n63.MaterialAttributes, Base Color=Multiply(A=@n63.Base Color, B=@n65)) {attributeSetTypes [#2]}
   False=@n62
  ) {parameterName=UseAO}
  n67 = ShadingPathSwitch(Default=@n66, Deferred=@n66, Forward=@n62, Mobile=@n62)
  n68 = VectorParameter() {bUseCustomPrimitiveData=true, group=z_code, parameterName=WSDeformationCenter, primitiveDataIndex=20}
  n69 = ScalarParameter() {group=z_code, parameterName=Code_DestructionAmount}
  n70 = StaticBoolParameter() {defaultValue=true, group=z_code, parameterName=Code_UseHitEffects}
  n71 = ScalarParameter() {bUseCustomPrimitiveData=true, defaultValue=1, group=z_code, parameterName=WSDamageRadius, primitiveDataIndex=26}
  n72 = ScalarParameter() {bUseCustomPrimitiveData=true, group=z_code, parameterName=DistanceCullFadeAnimationOffset, primitiveDataIndex=12}
  n73 = StaticSwitchParameter(
   True=SetMaterialAttributes!(
    MaterialAttributes=@n67
    World Position Offset=MaterialFunctionCall!(
     WSDeformationCenter=@n68.RGB
     Code_DestructionAmount=@n69
     Code_UseUvForDamage
     Code_UseHitEffects=@n70
     Code_Emissive
     WSDamageRadius=@n71
     CrackDamageShader
     Diffuse
     ScalarModDamageCracks
     ScalarOffsetforDamageCracks
     ExposeLerpFunctionForInvertingTheHitDistanceCalc
     InvertDamageFalloff
     CrackWorldPosition
     WorldPosition=@n72
     DistanceCullFadeAnimationOffset
    ) {materialFunction=/Game/Packages/jl_tests/Fortnite_HitEffects}.Hit And Shrink WPO
   ) {attributeSetTypes [#3]}
   False=@n67
  ) {parameterName=IsTeeth}
  Composite!(InputPin=@n73) {subgraphName=RVT}
  Composite!()
  Composite() {subgraphName=Bones}
  n74 = TextureSampleParameter2D() {
   parameterName=Basecolor
   samplerSource=SSM_Wrap_WorldGroupSettings
   texture=&T_Doggus_Sword_D
  }
  n75 = TextureSample() {samplerSource=SSM_Wrap_WorldGroupSettings, texture=&T_Doggus_Sword_E}
  n76 = TextureSample() {
   samplerSource=SSM_Wrap_WorldGroupSettings
   samplerType=SAMPLERTYPE_Masks
   texture=&T_Doggus_Sword_FX
  }
  n77 = TextureSampleParameter2D() {
   parameterName=Specular
   samplerSource=SSM_Wrap_WorldGroupSettings
   samplerType=SAMPLERTYPE_Masks
   texture=&T_Doggus_Sword_S
  }
  n78 = SetMaterialAttributes!(
   MaterialAttributes
   Base Color=StaticSwitchParameter(True=LinearInterpolate(A=@n74.RGB, B=@n75.RGB, Alpha=@n76.R), False=@n74.RGB) {defaultValue=true, parameterName=UseEmissiveBlend}
   Metallic=@n77.G
   Specular=@n77.G
   Roughness=LinearInterpolate(A=ScalarParameter() {parameterName=RoughnessMin}, B=ScalarParameter() {defaultValue=1, parameterName=RoughnessMax}, Alpha=@n77.B)
  ) {attributeSetTypes [
    #2
    #4
    #5
    #6
   ]}
  n79 = TextureSampleParameter2D() {
   parameterName=Mask
   samplerSource=SSM_Wrap_WorldGroupSettings
   samplerType=SAMPLERTYPE_Masks
   texture=&Cattus_Skull_Outside_M
  }
  n80 = LinearInterpolate(
   A=LinearInterpolate(
    A=VectorParameter() {group=Bone, parameterName=Color1, defaultValue=<rgba 0.3698 0.2375 0.089 1>}.RGB
    B=VectorParameter() {group=Bone, parameterName=Color2, defaultValue=<rgba 0.5747 0.5747 0.5747 1>}.RGB
    Alpha=@n79.G
   )
   B=VectorParameter() {group=Bone, parameterName=Color3, defaultValue=<rgba 0.1892 0.1685 0.1277 1>}.RGB
   Alpha=@n79.B
  )
  n81 = Fresnel(ExponentIn=ScalarParameter() {defaultValue=1.3, group=Bone, parameterName=FresnelBone}, Normal=VertexNormalWS()) {baseReflectFraction=0}
  n82 = TextureObject() {texture=&T_Bone_D}
  n83 = TextureObject() {samplerType=SAMPLERTYPE_Normal, texture=&T_Bone_N}
  n84 = MaterialFunctionCall!(
   Diffuse Texture Object=@n82
   Spec Texture Object=TextureObject() {samplerType=SAMPLERTYPE_Masks, texture=&T_Bone_M}
   TextureSize=ScalarParameter() {defaultValue=369.8548, group=Bone, parameterName=SizeBone_Tiling}
   --------------------------
   Use seperate settings for Z projection
   Use seperate texture for Z projection
   Texture Size Z projection
   Z Plane Diffuse Texture Object
   Z Plane Spec Texture Object
   "--------------------------      "
   ProjectionTransitionContrast=ScalarParameter() {defaultValue=2, group=Bone, parameterName=BoneProjectionContrast}
   Use Texture Alpha Heightmap to Refine the Transistion
   HeightLerpTransitionContrast
   PickTheDiffuseHeightChannel
   TheDiffuseHeightChannel
   "--------------------------       "
   NormalTextureObject=@n83
   World Space Normals=StaticBool()
   "--------------------------  "
   Export Float 4 Diffuse
   Export Float 4 Spec
   "--------------------------     "
   WorldPosition
   World Space Normal
  ) {materialFunction=/Engine/Functions/Engine_MaterialFunctions02/Texturing/WorldAlignedTextures_Complex}
  n85 = MaterialFunctionCall(Float3=@n84.XYZ Spec) {materialFunction=&BreakOutFloat3Components}
  n86 = Add(A=@n80, B=Multiply(A=Multiply(A=@n80, B=@n81), B=@n85.B))
  n87 = Multiply(A=TextureCoordinate(), B=ScalarParameter() {defaultValue=55, group=Bone, parameterName=Details UV Tile})
  n88 = TextureSample(UVs=@n87, Tex=@n82)
  n89 = NamedRerouteDeclaration(None=@n84.XYZ Diffuse Texture) {name=BoneD, nodeColor=<rgba 0 0.5781 1 1>}
  n90 = ComponentMask(None=VertexNormalWS()) {b=true}
  n91 = ScalarParameter() {defaultValue=0.4, parameterName=Bone Spec}
  n92 = Constant3Vector() {constant=<rgba 1 1 0 0>}
  n93 = SetMaterialAttributes!(
   MaterialAttributes
   Base Color=Multiply(A=ShadingPathSwitch(Default=@n86, Mobile=@n80), B=Multiply(A=@n88.R, B=@n89))
   Specular=Multiply(A=Max(A=@n90) {constB=0.25}, B=Multiply(A=@n91, B=@n85.R))
   Roughness=LinearInterpolate(
    A=ScalarParameter() {defaultValue=0.1, group=Bone, parameterName=RoughnessBoneMin}
    B=ScalarParameter() {defaultValue=1.2, group=Bone, parameterName=RoughnessBoneMax}
    Alpha=@n85.B
   )
   Normal=Add(A=Multiply(A=@n84.Normal XYZ Texture, B=@n92), B=TextureSample(UVs=@n87, Tex=@n83).RGB)
  ) {attributeSetTypes [
    #2
    #5
    #6
    #1
   ]}
  MaterialExpressionReroute_3 = Reroute(None=@n93)
  n94 = StaticSwitchParameter(True=@n78, False=@MaterialExpressionReroute_3.Bone Surf) {parameterName=IsSword}
  MaterialExpressionReroute_5 = Reroute(None=@n94)
  n95 = GetMaterialAttributes(None=@MaterialExpressionReroute_5.Input) {attributeGetTypes [#2]}
  n96 = VectorParameter() {group=Dust, parameterName=Dust, defaultValue=<rgba 0.7726 0.5853 0.4282 0>}
  n97 = NamedRerouteUsage()
  n98 = ScalarParameter() {defaultValue=2000, group=Dust, parameterName=Dust Distance}
  n99 = TextureSampleParameter2D() {
   parameterName=Mask
   samplerSource=SSM_Wrap_WorldGroupSettings
   samplerType=SAMPLERTYPE_Masks
   texture=&Cattus_Skull_Outside_M
  }
  n100 = Constant() {r=1.5}
  n101 = BlendMaterialAttributes(
   A=SetMaterialAttributes!(MaterialAttributes=@n95.MaterialAttributes, Base Color=Multiply(A=@n95.Base Color, B=@n96.RGB), Roughness=Constant() {r=0.75}) {attributeSetTypes [#2, #6]}
   B=@MaterialExpressionReroute_5.Input
   Alpha=MaterialFunctionCall(Transition Phase=Divide(A=@n97, B=@n98), Height Texture=Multiply(A=@n99.R, B=@n100)) {materialFunction=&HeightLerp}.Lerp Alpha No Contrast
  )
  MaterialExpressionReroute_2 = Reroute(None=@n101)
  Composite!(None=@MaterialExpressionReroute_2.DustLayer) {subgraphName=Tangent to World}
  MaterialExpressionReroute_9 = Reroute(None=@MaterialExpressionReroute_2.DustLayer)
  n102 = GetMaterialAttributes(None=@MaterialExpressionReroute_9.Input) {attributeGetTypes [#1]}
  n103 = SetMaterialAttributes!(MaterialAttributes=@n102.MaterialAttributes, Normal=Transform(None=@n102.Normal)) {attributeSetTypes [#1]}
  MaterialExpressionReroute_10 = Reroute(None=@n103)
  Composite!(MaterialAttributes=@MaterialExpressionReroute_10.N Tangent To World) {subgraphName=WPO Bounce}
  Composite!(None=@n94) {subgraphName=Dust}
  MaterialExpressionReroute_11 = Reroute(None=@MaterialExpressionReroute_10.N Tangent To World)
  n104 = MaterialFunctionCall(
   WSDeformationCenter=VectorParameter() {bUseCustomPrimitiveData=true, group=z_code, parameterName=WSDeformationCenter, primitiveDataIndex=20}.RGB
   Code_DestructionAmount=ScalarParameter() {group=z_code, parameterName=Code_DestructionAmount}
   Code_UseHitEffects=StaticBoolParameter() {defaultValue=true, group=z_code, parameterName=Code_UseHitEffects}
   WSDamageRadius=ScalarParameter() {bUseCustomPrimitiveData=true, defaultValue=1, group=z_code, parameterName=WSDamageRadius, primitiveDataIndex=26}
   DistanceCullFadeAnimationOffset=ScalarParameter() {bUseCustomPrimitiveData=true, group=z_code, parameterName=DistanceCullFadeAnimationOffset, primitiveDataIndex=12}
  ) {materialFunction=&Fortnite_HitEffects}
  MaterialExpressionReroute_0 = Reroute(None=@n104.Hit And Shrink WPO)
  n105 = SetMaterialAttributes!(MaterialAttributes=@MaterialExpressionReroute_11.MaterialAttributes, World Position Offset=@MaterialExpressionReroute_0.Hit And Shrink WPO) {attributeSetTypes [#3]}
  MaterialExpressionReroute_12 = Reroute(None=@n105)
  n106 = StaticSwitchParameter(True=@MaterialExpressionReroute_12.WPO Wobble, False=@MaterialExpressionReroute_10.N Tangent To World) {parameterName=IsTeeth}
  Composite!(InputPin=@n106) {subgraphName=RVT}
  Constant3Vector()
  n107 = MaterialFunctionCall(Float3=@n58.Normal) {materialFunction=&BreakOutFloat3Components}
  n108 = Multiply(A=Transform(None=Constant3Vector() {constant=<rgba 0 0 1 0>}), B=Constant() {r=1})
  n109 = AppendVector(A=Multiply(A=@n107.R) {constB=-1}, B=Multiply(A=@n107.B, B=ComponentMask(None=@n108) {g=true}))
  n110 = Multiply(A=@n107.G) {constB=-1}
  n111 = ComponentMask(None=@n108) {r=true}
  n112 = AppendVector(A=Multiply(A=@n111, B=@n107.B), B=Multiply(A=@n107.R) {constB=-1})
  n113 = Multiply(A=@n107.G) {constB=-1}
  n114 = Abs(None=@n111)
  n115 = ScalarParameter() {parameterName=ContrastBlendNormals}
  n116 = Multiply(A=@n58.Normal, B=Constant3Vector() {constant=<rgba -1 -1 1 0>})
  n117 = ComponentMask(None=@n108) {b=true}
  LinearInterpolate(
   A=LinearInterpolate(
    A=AppendVector(A=@n109, B=@n110)
    B=AppendVector(A=@n112, B=@n113)
    Alpha=MaterialFunctionCall(In=@n114, Contrast=@n115) {materialFunction=&CheapContrast}.Result
   )
   B=AppendVector(A=ComponentMask(None=@n116) {g=true, r=true}, B=Multiply(A=@n107.B, B=@n117))
   Alpha=MaterialFunctionCall(In=Abs(None=@n117), Contrast=@n115) {materialFunction=&CheapContrast}.Result
  )
  MaterialFunctionCall(BaseNormal=@n3.Normal XYZ Texture, AdditionalNormal=@n2.Normal XYZ Texture) {materialFunction=&BlendAngleCorrectedNormals}
  MaterialFunctionCall(NoRVT=@MaterialExpressionReroute_5.Input, RVT=@n101) {materialFunction=&FN_RVT_Switch_Material}
  MaterialFunctionCall(Normal=@n35.Normal XY Texture, Flatness=ScalarParameter() {parameterName=NormalIntensityGrass}) {materialFunction=&FlattenNormal}
  MaterialProxyReplace(Realtime=VertexInterpolator(VS=@n34).PS, MaterialProxy=@n34)
  MaterialProxyReplace(Realtime=VertexInterpolator(VS=@n14).PS, MaterialProxy=@n14)
  NamedRerouteDeclaration(None=@n57.Blended Material) {name=Mat, nodeColor=<rgba 0.875 1 0 1>}
  NamedRerouteUsage()
  NamedRerouteUsage()
  Normalize(VectorInput=@n4)
  n118 = RuntimeVirtualTextureSampleParameter() {materialType=BaseColor_Normal_Specular_YCoCg, parameterName=RVT, virtualTexture=/Game/Athena/Apollo/Maps/Landscape/VT/RVT_Apollo_Landscape}
  MaterialExpressionReroute_20 = Reroute(None=@n73)
  n119 = RuntimeVirtualTextureSampleParameter() {materialType=WorldHeight, parameterName=RVTHeight, virtualTexture=/Game/Athena/Apollo/Maps/Landscape/VT/RVT_Apollo_Landscape_Height}
  n120 = Subtract(A=ComponentMask(None=WorldPosition().XYZ) {b=true}, B=@n119.WorldHeight)
  n121 = NamedRerouteDeclaration(None=NamedRerouteDeclaration(None=@n120) {name=RVT Raw, nodeColor=<rgba 1 0 0.4219 1>}) {name=RVT Height, nodeColor=<rgba 0.2422 1 0 1>}
  n122 = ScalarParameter() {defaultValue=8, group=RVT, parameterName=Blend RVT Normals}
  n123 = BlendMaterialAttributes(
   A=SetMaterialAttributes!(MaterialAttributes=@MaterialExpressionReroute_20.InputPin, Normal=@n118.Normal) {attributeSetTypes [#1]}
   B=@MaterialExpressionReroute_20.InputPin
   Alpha=Saturate(None=Divide(A=@n121, B=@n122))
  ) {vertexAttributeBlendType=UseA}
  n124 = SetMaterialAttributes!(MaterialAttributes=@n123, Base Color=@n118.BaseColor, Specular=@n118.BaseColor, Roughness=@n118.BaseColor, Emissive Color=Constant()) {attributeSetTypes [
    #2
    #5
    #6
    #7
   ]}
  n125 = ScalarParameter() {defaultValue=100, group=RVT, parameterName=Blend RVT}
  n126 = BlendMaterialAttributes(A=@n124, B=@n123, Alpha=Saturate(None=Divide(A=@n120, B=@n125))) {vertexAttributeBlendType=UseA}
  PinBase!(Output=@n126) {reroutePins [Output {expression=&self:MaterialExpressionReroute_30}]}
  PinBase!() {reroutePins [InputPin {expression=&self:MaterialExpressionReroute_20}]}
  PinBase() {reroutePins [MaterialAttributes {expression=&self:MaterialExpressionReroute_11}]}
  PinBase!(WPO Wobble=@n105) {reroutePins [WPO Wobble {expression=&self:MaterialExpressionReroute_12}]}
  PinBase!() {reroutePins [Input {expression=&self:MaterialExpressionReroute_5}]}
  PinBase!(DustLayer=@n101) {reroutePins [DustLayer {expression=&self:MaterialExpressionReroute_2}]}
  PinBase!() {reroutePins [InputPin {expression=&self:MaterialExpressionReroute_24}]}
  n127 = RuntimeVirtualTextureSampleParameter() {materialType=BaseColor_Normal_Specular_YCoCg, parameterName=RVT, virtualTexture=&RVT_Apollo_Landscape}
  MaterialExpressionReroute_24 = Reroute(None=@n106)
  n128 = RuntimeVirtualTextureSampleParameter() {materialType=WorldHeight, parameterName=RVTHeight, virtualTexture=&RVT_Apollo_Landscape_Height}
  n129 = Subtract(A=ComponentMask(None=WorldPosition().XYZ) {b=true}, B=@n128.WorldHeight)
  n130 = NamedRerouteDeclaration(None=NamedRerouteDeclaration(None=@n129) {name=RVT Raw, nodeColor=<rgba 1 0 0.4219 1>}) {name=RVT Height, nodeColor=<rgba 0.2422 1 0 1>}
  n131 = ScalarParameter() {defaultValue=8, group=RVT, parameterName=Blend RVT Normals}
  n132 = BlendMaterialAttributes(
   A=SetMaterialAttributes!(MaterialAttributes=@MaterialExpressionReroute_24.InputPin, Normal=@n127.Normal) {attributeSetTypes [#1]}
   B=@MaterialExpressionReroute_24.InputPin
   Alpha=Saturate(None=Divide(A=@n130, B=@n131))
  ) {vertexAttributeBlendType=UseA}
  n133 = SetMaterialAttributes!(MaterialAttributes=@n132, Base Color=@n127.BaseColor, Specular=@n127.BaseColor, Roughness=@n127.BaseColor, Emissive Color=Constant()) {attributeSetTypes [
    #2
    #5
    #6
    #7
   ]}
  n134 = ScalarParameter() {defaultValue=100, group=RVT, parameterName=Blend RVT}
  n135 = BlendMaterialAttributes(A=@n133, B=@n132, Alpha=Saturate(None=Divide(A=@n129, B=@n134))) {vertexAttributeBlendType=UseA}
  PinBase!(Output=@n135) {reroutePins [Output {expression=&self:MaterialExpressionReroute_23}]}
  PinBase!()
  PinBase!(Hit And Shrink WPO=@n104.Hit And Shrink WPO) {reroutePins [Hit And Shrink WPO {expression=&self:MaterialExpressionReroute_0}]}
  PinBase!()
  PinBase!(Bone Surf=@n93) {reroutePins [Bone Surf {expression=&self:MaterialExpressionReroute_3}]}
  PinBase!() {reroutePins [Input {expression=&self:MaterialExpressionReroute_9}]}
  PinBase!(N Tangent To World=@n103) {reroutePins [N Tangent To World {expression=&self:MaterialExpressionReroute_10}]}
  Reroute(None=@n60.RGB)
  Reroute(None=@n84.XYZ Diffuse Texture)
  ScalarParameter() {defaultValue=5500, parameterName=BlendDistance}
  ScalarParameter() {defaultValue=8000, parameterName=SwordBlendDistance}
  ScalarParameter() {defaultValue=200, parameterName=SwordBlendOffset}
  n136 = GetMaterialAttributes(None=@n94) {attributeGetTypes [#1, #2]}
  n137 = Constant3Vector() {constant=<rgba 1 1 0 0>}
  n138 = TextureSampleParameter2D() {parameterName=Baked Normal, samplerType=SAMPLERTYPE_Normal, texture=&T_Doggus_Sword_N}
  n139 = TextureSampleParameter2D() {
   parameterName=Mask
   samplerSource=SSM_Wrap_WorldGroupSettings
   samplerType=SAMPLERTYPE_Masks
   texture=&Cattus_Skull_Outside_M
  }
  SetMaterialAttributes!(
   MaterialAttributes=@n136.MaterialAttributes
   Normal=Add(A=Multiply(A=@n136.Normal, B=@n137), B=@n138.RGB)
   Base Color=LinearInterpolate(A=@n136.Base Color, B=Multiply(A=@n136.Base Color, B=@n139.RGB), Alpha=ScalarParameter() {parameterName=BaseColorMaskMulti_Alpha})
  ) {attributeSetTypes [#1, #2]}
  n140 = MaterialFunctionCall(MA In=@n106) {materialFunction=&MF_CreativeMoveTool}
  MaterialExpressionReroute_23 = Reroute(None=@n135)
  ShadingPathSwitch(
   Default=MaterialFunctionCall(NoRVT=@n140.MA Out, RVT=MaterialProxyReplace(Realtime=@MaterialExpressionReroute_23.Output, MaterialProxy=@n106)) {materialFunction=&FN_RVT_Switch_Material}.Result
   Mobile=@n140.MA Out
  )
  n141 = MaterialFunctionCall(MA In=@n73) {materialFunction=&MF_CreativeMoveTool}
  MaterialExpressionReroute_30 = Reroute(None=@n126)
  ShadingPathSwitch(
   Default=MaterialFunctionCall(NoRVT=@n141.MA Out, RVT=MaterialProxyReplace(Realtime=@MaterialExpressionReroute_30.Output, MaterialProxy=@n73)) {materialFunction=&FN_RVT_Switch_Material}.Result
   Mobile=@n141.MA Out
  )
  Subtract(A=@n31) {constB=768}
  Transform(None=Constant3Vector() {constant=<rgba 0 0 1 0>}) {transformSourceType=TRANSFORMSOURCE_Local}
 ]
}
